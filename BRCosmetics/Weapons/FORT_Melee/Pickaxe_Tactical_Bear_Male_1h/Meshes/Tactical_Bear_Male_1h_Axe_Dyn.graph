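# Physics asset for the one-handed Tactical Bear Male pickaxe mesh, providing bodies for its dynamic fin and tail bones.
# A small capsule body sits on the handle bone, with box bodies on the left fin, right fin and tail.
# Each appendage is constrained directly to the handle with limited swing and tight twist.
# The two fins share identical limits and swing more widely than the larger tail.
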PhysicsAsset /BRCosmetics/Weapons/FORT_Melee/Pickaxe_Tactical_Bear_Male_1h/Meshes/Tactical_Bear_Male_1h_Axe_Dyn
bodies {
 Melee_TwoHanded_Handle [{shapeType=Capsule, length=1, radius=1}]
 dyn_fin_l [{shapeType=Box, extentX=6, extentY=5, extentZ=2, center=<3 0 0>}]
 dyn_fin_r [{shapeType=Box, extentX=6, extentY=5, extentZ=2, center=<3 0 0>}]
 dyn_tail [{shapeType=Box, extentX=12, extentY=9, extentZ=3, center=<3 0 0>}]
}
constraints [
 {bone1Name=dyn_fin_l, bone2Name=Melee_TwoHanded_Handle, swing1LimitDegrees=20, swing2LimitDegrees=45, twistLimitDegrees=5}
 {bone1Name=dyn_fin_r, bone2Name=Melee_TwoHanded_Handle, swing1LimitDegrees=20, swing2LimitDegrees=45, twistLimitDegrees=5}
 {bone1Name=dyn_tail, bone2Name=Melee_TwoHanded_Handle, swing1LimitDegrees=5, swing2LimitDegrees=30, twistLimitDegrees=5}
]
