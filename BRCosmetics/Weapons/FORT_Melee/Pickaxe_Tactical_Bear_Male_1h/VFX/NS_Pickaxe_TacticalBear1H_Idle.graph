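# Idle Niagara effect for the one-handed Tactical Bear pickaxe cosmetic, layering cool blue glints, sparkles and drifting smoke along the weapon.
# NE_Bling and NE_Sparkles spawn sprites in local offset boxes, fading or scaling them in and out over short lifetimes.
# NE_Smoke spawns meshes in a cylinder that sink slowly and stretch, with alpha and dynamic material parameters animated over life.
# Pickaxe-state user parameters are declared, but no shown module reads them.
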NiagaraSystem /BRCosmetics/Weapons/FORT_Melee/Pickaxe_Tactical_Bear_Male_1h/VFX/NS_Pickaxe_TacticalBear1H_Idle
dependencies {
 usedDataInterfaces [&NiagaraDataInterfaceVectorCurve, &NiagaraDataInterfaceCurve]
 usedDynamicInputs []
 usedModules [&SystemState]
}
emitters {
 NE_Bling {
  emitterUpdateScript [
   &EmitterState()
   &SpawnRate(SpawnRate=&RandomRangeFloat(Minimum=4, Maximum=12), Use Spawn Probability=false)
  ]
  particleSpawnScript [&InitializeParticle(
    Color=&RandomRangeLinearColor(Minimum=<rgba 0.3 0.6 1 0.5>, Maximum=<rgba 0.3 0.6 1 1>)
    Lifetime Max=0.2
    Lifetime Min=0.1
    Lifetime Mode=::NewEnumerator1
    Position Offset=&RandomRangeVector(Minimum=<-10 0 10>, Maximum=<10 0 45>)
    Position Offset Coordinate Space=::Local
    Sprite Rotation Mode=::NewEnumerator2
    Sprite Size Mode=::NewEnumerator1
    Uniform Sprite Size Max=6
    Uniform Sprite Size Min=4
    UsePositionOffset=true
    Write Material Random=true
   )]
  particleUpdateScript [
   &ParticleState()
   &SpriteRotationRate(Rotation Rate=&RandomRangeFloat(Minimum=-0.125, Maximum=0.125, Evaluation Type=::NewEnumerator1))
   &ScaleColor(
    Scale Alpha=&FloatFromCurve(FloatCurve=DataInterface() {propertyValues {Curve {keys [<at 0 0>,<at 0.35 1>,<at 1 0>]}}})
    ScaleA=true
    ScaleRGB=true
   )
   &CameraOffset(Camera Offset Amount=15)
  ]
  rendererClasses [&NiagaraSpriteRendererProperties]
  renderers [{rendererClass=&NiagaraSpriteRendererProperties}]
 }
 NE_Smoke {
  emitterUpdateScript [
   &EmitterState()
   &SpawnRate(SpawnRate=4, Use Spawn Probability=false)
  ]
  particleSpawnScript [
   &InitializeParticle(
    Color Maximum=<rgba 0 1 1.5 13>
    Color Minimum=<rgba 1 1 1 13>
    Color Mode=::NewEnumerator2
    Lifetime Max=3
    Lifetime Min=2
    Lifetime Mode=::NewEnumerator1
    Mesh Scale Max=<0.08 0.025 0.1>
    Mesh Scale Min=<0.06 0.01 0.1>
    Mesh Scale Mode=::NewEnumerator2
    Position Offset=<3 0 15>
    Position Offset Coordinate Space=::Local
    UsePositionOffset=true
    Write Material Random=true
   )
   &ShapeLocation(
    Cylinder Height=60
    Cylinder Height Midpoint=0
    Cylinder Radius=3
    MaskSpawnGroup=false
    Non Uniform Scale=<1 1 0.5>
    Override Seed=false
    Rotation Mode=::NewEnumerator2
    Shape Primitive=::NewEnumerator1
    Surface Only=false
    UseOwnerScale=true
   )
   &AddVelocity(
    Rotation Coordinate Space=::World
    Velocity=&RandomRangeVector(Minimum=<0 0 -2>, Maximum=<0 0 -5>, Evaluation Type=::NewEnumerator1)
   )
   &DynamicMaterialParameters(
    Index 0 Param 2=&RandomRangeFloat(Minimum=30, Maximum=50, Evaluation Type=::NewEnumerator1)
    Module.Index0Float0Write=false
    Module.Index0Float1Write=true
    Module.Index0Float2Write=false
    Module.Index0Float3Write=false
   )
  ]
  particleUpdateScript [
   &ParticleState()
   &ScaleMeshSize(Scale Factor=&VectorFromCurve(VectorCurve=DataInterface() {propertyValues {XCurve {keys [<at 0 0.1>,<at 1 10>]}, YCurve {keys [<at 0 1>,<at 1 8>]}, ZCurve {keys [<at 0 1>,<at 1 1>]}}}))
   &ScaleColor(
    Scale Alpha=&FloatFromCurve(FloatCurve=DataInterface() {propertyValues {Curve {keys [
         {arriveTangent=0.2972, interpMode=RCIM_Cubic, leaveTangent=0.2972, tangentMode=RCTM_User, time=0, value=-0.0092}
         {arriveTangent=2.8868, interpMode=RCIM_Cubic, leaveTangent=2.8868, tangentMode=RCTM_User, time=0.2459, value=1}
         {arriveTangent=-0.6384, interpMode=RCIM_Cubic, leaveTangent=-0.6384, time=0.3971, value=1.1998}
         {arriveTangent=-2.5872, interpMode=RCIM_Cubic, leaveTangent=-2.5872, time=0.6259, value=0.7574}
         {arriveTangent=-2.1791, interpMode=RCIM_Cubic, leaveTangent=-2.1791, time=0.7543, value=0.2756}
         {arriveTangent=-1.1217, interpMode=RCIM_Cubic, leaveTangent=-1.1217, time=0.9211, value=0.1142}
         <at 1 0 RCIM_Cubic>
        ]}}})
    ScaleA=true
    ScaleRGB=true
   )
   &Drag(Use Linear Drag=true, Use Rotational Drag=true)
   &SolveForcesAndVelocity(Clamp Velocity=false, Limit Acceleration=false)
   &CameraOffset(Camera Offset Amount=10)
   &DynamicMaterialParameters(
    Index 0 Param 1=&FloatFromCurve(FloatCurve=DataInterface() {propertyValues {Curve {keys [<at 0 5>,<at 1 0>]}}})
    Index 0 Param 4=&FloatFromCurve(FloatCurve=DataInterface() {propertyValues {Curve {keys [
         {arriveTangent=-0.0169, interpMode=RCIM_Cubic, leaveTangent=-0.0169, tangentMode=RCTM_User, time=0.0123, value=0.25}
         {arriveTangent=-0.3638, interpMode=RCIM_Cubic, leaveTangent=-0.3638, tangentMode=RCTM_User, time=0.8212, value=0.2237}
         {arriveTangent=-2.1295, interpMode=RCIM_Cubic, leaveTangent=-2.1295, tangentMode=RCTM_User, time=0.9123, value=0.1}
         <at 1 0 RCIM_Cubic>
        ]}}})
    Module.Index0Float0Write=true
    Module.Index0Float1Write=false
    Module.Index0Float2Write=true
    Module.Index0Float3Write=true
   ) {module=DynamicMaterialParameters001}
  ]
  rendererClasses [&NiagaraMeshRendererProperties]
  renderers [{rendererClass=&NiagaraMeshRendererProperties, rendererIndex=0}]
 }
 NE_Sparkles {
  emitterUpdateScript [
   &EmitterState()
   &SpawnRate(SpawnRate=&RandomRangeFloat(Minimum=1, Maximum=2), Use Spawn Probability=false)
  ]
  particleSpawnScript [&InitializeParticle(
    Color=&RandomRangeLinearColor(Minimum=<rgba 0.1 1.3667 2 0.5>, Maximum=<rgba 0.1 1.3667 2 1.25>)
    Lifetime Max=0.8
    Lifetime Min=0.35
    Lifetime Mode=::NewEnumerator1
    Position Offset=&RandomRangeVector(Minimum=<-8 0 10>, Maximum=<8 0 45>)
    Position Offset Coordinate Space=::Local
    Sprite Size Mode=::NewEnumerator1
    Uniform Sprite Size Max=25
    Uniform Sprite Size Min=16
    UsePositionOffset=true
    Write Material Random=true
   )]
  particleUpdateScript [
   &ParticleState()
   &SpriteRotationRate(Rotation Rate=&RandomRangeFloat(Minimum=-0.125, Maximum=0.125, Evaluation Type=::NewEnumerator1))
   &ScaleSpriteSize(
    Scale Sprite Size Mode=::NewEnumerator0
    Uniform Scale Factor=&FloatFromCurve(FloatCurve=DataInterface() {propertyValues {Curve {keys [<at 0 0>,<at 0.175 1>,<at 0.6 1>,<at 1 0>]}}})
   )
   &CameraOffset(Camera Offset Amount=5)
   &DynamicMaterialParameters(
    Index 0 Param 1=0.4
    Index 0 Param 2=1
    Index 0 Param 3=1
    Index 0 Param 4=1
    Module.Index0Float0Write=true
    Module.Index0Float1Write=true
    Module.Index0Float2Write=true
    Module.Index0Float3Write=true
   )
  ]
  rendererClasses [&NiagaraSpriteRendererProperties]
  renderers [{rendererClass=&NiagaraSpriteRendererProperties}]
 }
}
summary {
 systemName=NS_Pickaxe_TacticalBear1H_Idle
 emitters [
  {emitterName=NE_Smoke, rendererClasses [&NiagaraMeshRendererProperties]}
  {emitterName=NE_Bling, rendererClasses [&NiagaraSpriteRendererProperties]}
  {emitterName=NE_Sparkles, rendererClasses [&NiagaraSpriteRendererProperties]}
 ]
 userVariables [
  User.bIsFrontEnd {type=NiagaraBool}
  User.bIsFrontEndPreview {type=NiagaraBool}
  User.bisSwinging {type=NiagaraBool}
  User.IsFlying {type=NiagaraFloat, defaultValue=0, description=""}
  User.IsSkydiving {type=NiagaraBool}
  User.PhysicalSurface {type=NiagaraInt32}
  User.RibbonIDOffset {type=NiagaraInt32}
  User.SwingSectionIndex {type=NiagaraInt32}
  User.VariantID {type=NiagaraFloat, defaultValue=0, description=""}
 ]
}
